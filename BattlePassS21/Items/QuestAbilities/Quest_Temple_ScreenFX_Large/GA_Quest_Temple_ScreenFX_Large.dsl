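; Gameplay ability for a Battle Pass Season 21 Temple quest that plays a large psychedelic screen effect when the player interacts with an Object of Power.
; On event activation it gets the owning pawn's player controller after calling the parent.
; It then starts the Temple psychedelic camera shake in camera-local space, plays a configurable force feedback effect under a named tag, and ends itself immediately.
; The end-ability event is empty.
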(event Ability|EventActivateAbilityFromEvent
  (params (EventData :type Gameplay%20Event%20Data%20Structure%20%28by%20ref%29))
  (bind _owningplayerpawn (Variables|GAVars|Data|GetOwningPlayerPawn))
  (bind _returnvalue (Fortnite|GetFortPlayerControllerfromActor _owningplayerpawn))
  ;[1 1 1] Activate ability - listen for when player interacts with an Object of Power
  (|CallParentFunction EventData)
  ;;[1 1 1] Play Camera Anim of Psychedelic Effect
  (Game|Feedback|ClientStartCameraShake :self _returnvalue :Shake "/Game/Effects/Fort_Effects/Camera_Effects/BP_CameraShake_S21_TemplePsychedelicEffect.BP_CameraShake_S21_TemplePsychedelicEffect_C" :Scale 1.0 :PlaySpace "CameraLocal")
  (Game|Feedback|ClientPlayForceFeedback _returnvalue (Variables|Default|GetForce%20Feedback%20Effect) "TempleFullScreenForceFeedback" false false false)
  (Ability|EndAbility))

(event Ability|EventOnEndAbility
  (params (bWasCancelled :type Boolean)))
